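# StrawTurnHead_Equip_Cue is the sound cue for equipping the StrawTurn head, layering an equip sound with a laugh.
# A mixer combines one of four random equip waves, with slight volume variation, and a single laugh wave delayed briefly.
# It uses weapon-reload attenuation and the quests sound class, with concurrency settings limiting overlapping instances.
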SoundCue /StrawTurn/Audio/StrawTurnHead_Equip_Cue
root=Mixer_0
nodes [
 {id=Delay_0, children [WavePlayer_8], properties {delayMax=0.1, delayMin=0.1}}
 {id=Mixer_0, children [Modulator_0,Delay_0], properties {inputVolume [1,1]}}
 {id=Modulator_0, children [Random_1], properties {pitchMax=1, pitchMin=1, volumeMax=1, volumeMin=0.85}}
 {
  id=Random_1
  children [WavePlayer_0,WavePlayer_2,WavePlayer_3,WavePlayer_7]
  properties {bShouldExcludeFromBranchCulling=true, weights [1,1,1,1]}
 }
 {id=WavePlayer_0, properties {soundWaveAssetPtr=/StrawTurn/Audio/Waves/CH5S4_Sawturn_Equip_01}}
 {id=WavePlayer_2, properties {soundWaveAssetPtr=/StrawTurn/Audio/Waves/CH5S4_Sawturn_Equip_02}}
 {id=WavePlayer_3, properties {soundWaveAssetPtr=/StrawTurn/Audio/Waves/CH5S4_Sawturn_Equip_03}}
 {id=WavePlayer_7, properties {soundWaveAssetPtr=/StrawTurn/Audio/Waves/CH5S4_Sawturn_Equip_04}}
 {id=WavePlayer_8, properties {soundWaveAssetPtr=/StrawTurn/Audio/Waves/CH5S4_StrawTurn_Laugh_01}}
]
properties {
 attenuationSettings=&Fort_Weapon_Reload_Att
 bOverrideConcurrency=true
 maxDistance=2250
 soundClassObject=&Fort_Quests
 volumeMultiplier=0.7
 attenuationOverrides {falloffDistance=1600}
 concurrencyOverrides {maxCount=2}
 concurrencySet [/Game/Sounds/Concurrency_Rules/OneMaximum_Owner_StopOldestRelease]
}
