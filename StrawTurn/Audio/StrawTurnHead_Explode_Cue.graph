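# Sound cue for the StrawTurn head explosion, using the explosives sound class and its own explosion attenuation.
# A distance crossfade blends close and distant layers, with close fading out as distant fades in over nearly the same range.
# Each layer passes through a modulator into an equally weighted random pick of five waves.
# The close layer's random node is excluded from branch culling, unlike the distant one.
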SoundCue /StrawTurn/Audio/StrawTurnHead_Explode_Cue
root=DistanceCrossFade_0
nodes [
 {
  id=DistanceCrossFade_0
  children [Modulator_0,Modulator_1]
  properties {crossFadeInput [
    {fadeInDistanceEnd=0, fadeInDistanceStart=0, fadeOutDistanceEnd=15000, fadeOutDistanceStart=5000, volume=1}
    {fadeInDistanceEnd=14000, fadeInDistanceStart=4700, fadeOutDistanceEnd=0, fadeOutDistanceStart=0, volume=1}
   ]}
 }
 {id=Modulator_0, children [Random_0]}
 {id=Modulator_1, children [Random_1]}
 {
  id=Random_0
  children [WavePlayer_10, WavePlayer_11, WavePlayer_12, WavePlayer_13, WavePlayer_14]
  properties {bShouldExcludeFromBranchCulling=true, weights [1,1,1,1,1]}
 }
 {id=Random_1, children [WavePlayer_15, WavePlayer_16, WavePlayer_17, WavePlayer_18, WavePlayer_19], properties {weights [1,1,1,1,1]}}
 {id=WavePlayer_10, properties {soundWaveAssetPtr=/StrawTurn/Audio/Waves/CH5S4_Sawturn_Explode_Close_01}}
 {id=WavePlayer_11, properties {soundWaveAssetPtr=/StrawTurn/Audio/Waves/CH5S4_Sawturn_Explode_Close_02}}
 {id=WavePlayer_12, properties {soundWaveAssetPtr=/StrawTurn/Audio/Waves/CH5S4_Sawturn_Explode_Close_03}}
 {id=WavePlayer_13, properties {soundWaveAssetPtr=/StrawTurn/Audio/Waves/CH5S4_Sawturn_Explode_Close_04}}
 {id=WavePlayer_14, properties {soundWaveAssetPtr=/StrawTurn/Audio/Waves/CH5S4_Sawturn_Explode_Close_05}}
 {id=WavePlayer_15, properties {soundWaveAssetPtr=/StrawTurn/Audio/Waves/CH5S4_Sawturn_Explode_Distant_01}}
 {id=WavePlayer_16, properties {soundWaveAssetPtr=/StrawTurn/Audio/Waves/CH5S4_Sawturn_Explode_Distant_02}}
 {id=WavePlayer_17, properties {soundWaveAssetPtr=/StrawTurn/Audio/Waves/CH5S4_Sawturn_Explode_Distant_03}}
 {id=WavePlayer_18, properties {soundWaveAssetPtr=/StrawTurn/Audio/Waves/CH5S4_Sawturn_Explode_Distant_04}}
 {id=WavePlayer_19, properties {soundWaveAssetPtr=/StrawTurn/Audio/Waves/CH5S4_Sawturn_Explode_Distant_05}}
]
properties {
 attenuationSettings=/StrawTurn/Audio/ATT/ATT_Sawturn_Expl
 maxDistance=18075
 priority=2
 soundClassObject=/Game/Sounds/FortSoundClassesAndModes/Fort_Explosives
 volumeMultiplier=0.3
 concurrencySet [null]
}
