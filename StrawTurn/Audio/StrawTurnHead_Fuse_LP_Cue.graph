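# Sound cue for the StrawTurn head's fuse, layering an ignition one-shot over a looping burn, routed to the quests sound class.
# A mixer blends a randomly chosen ignite variant, weighted well above the loop, with the looping wave.
# The loop starts after a brief delay and fades in through an enveloper, with slight volume modulation on each layer.
# Playback uses weapon-reload attenuation and overrides concurrency to cap simultaneous instances.
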SoundCue /StrawTurn/Audio/StrawTurnHead_Fuse_LP_Cue
root=Mixer_0
nodes [
 {id=Delay_0, children [WavePlayer_1], properties {delayMax=0.1, delayMin=0.1}}
 {
  id=Enveloper_0
  children [Modulator_0]
  properties {pitchCurve {editorCurveData {keys [{time=0}]}}, volumeCurve {editorCurveData {keys [<at 0 0>,<at 1 1>]}}}
 }
 {id=Mixer_0, children [Modulator_1,Enveloper_0], properties {inputVolume [3,0.5]}}
 {id=Modulator_0, children [Delay_0], properties {pitchMax=1, pitchMin=1, volumeMax=1, volumeMin=0.85}}
 {id=Modulator_1, children [Random_0], properties {volumeMin=0.98}}
 {id=Random_0, children [WavePlayer_0,WavePlayer_2,WavePlayer_3], properties {weights [1,1,1]}}
 {id=WavePlayer_0, properties {soundWaveAssetPtr=/StrawTurn/Audio/Waves/CH5S4_StrawTurn_Fuse_Ignite_01}}
 {id=WavePlayer_1, properties {bLooping=true, soundWaveAssetPtr=/StrawTurn/Audio/Waves/CH5S4_StrawTurn_Fuse_LP}}
 {id=WavePlayer_2, properties {soundWaveAssetPtr=/StrawTurn/Audio/Waves/CH5S4_StrawTurn_Fuse_Ignite_02}}
 {id=WavePlayer_3, properties {soundWaveAssetPtr=/StrawTurn/Audio/Waves/CH5S4_StrawTurn_Fuse_Ignite_03}}
]
properties {
 attenuationSettings=&Fort_Weapon_Reload_Att
 bOverrideConcurrency=true
 maxDistance=2250
 soundClassObject=&Fort_Quests
 volumeMultiplier=0.15
 attenuationOverrides {falloffDistance=1600}
 concurrencyOverrides {maxCount=2}
 concurrencySet [/Game/Sounds/Concurrency_Rules/OneMaximum_Owner_StopOldestRelease]
}
