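# Sound cue for the StrawTurn head's fuse warning, a quiet positional sound in the quests sound class.
# A Random node picks one of three equally weighted warning waves and feeds a Modulator that slightly varies volume while leaving pitch fixed.
# It borrows the weapon reload attenuation asset and overrides concurrency to cap playback at two instances, though a one-per-owner concurrency set is also referenced.
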SoundCue /StrawTurn/Audio/StrawTurnHead_Fuse_Warning_Cue
root=Modulator_0
nodes [
 {id=Modulator_0, children [Random_0], properties {pitchMax=1, pitchMin=1, volumeMax=1, volumeMin=0.85}}
 {id=Random_0, children [WavePlayer_0,WavePlayer_2,WavePlayer_3], properties {weights [1,1,1]}}
 {id=WavePlayer_0, properties {soundWaveAssetPtr=/StrawTurn/Audio/Waves/CH5S4_StrawTurn_Fuse_Warning_03}}
 {id=WavePlayer_2, properties {soundWaveAssetPtr=/StrawTurn/Audio/Waves/CH5S4_StrawTurn_Fuse_Warning_02}}
 {id=WavePlayer_3, properties {soundWaveAssetPtr=/StrawTurn/Audio/Waves/CH5S4_StrawTurn_Fuse_Warning_01}}
]
properties {
 attenuationSettings=&Fort_Weapon_Reload_Att
 bOverrideConcurrency=true
 maxDistance=2250
 soundClassObject=&Fort_Quests
 volumeMultiplier=0.15
 attenuationOverrides {falloffDistance=1600}
 concurrencyOverrides {maxCount=2}
 concurrencySet [/Game/Sounds/Concurrency_Rules/OneMaximum_Owner_StopOldestRelease]
}
