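# Sound cue for the StrawTurn job board "Accept" voice-over, classed under the Fort_Quests sound class.
# A single wave player feeds one accept recording straight to the output, with no randomization or variation.
# It borrows a TV prop attenuation asset and overrides concurrency so only one instance plays, preventing new triggers from interrupting it.
# The debug flag is left enabled.
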SoundCue /StrawTurn/Audio/StrawTurn_JobBoard_Accept_VO_Cue
root=WavePlayer_0
nodes [{id=WavePlayer_0, properties {soundWaveAssetPtr=/StrawTurn/Audio/Waves/CH5S4_Sawtrurn_JobBoard_Accept_01}}]
properties {
 attenuationSettings=/Game/Sounds/Attenuation/Fort_Prop_TV_Athena_Att
 bDebug=true
 bOverrideConcurrency=true
 maxDistance=500
 soundClassObject=&Fort_Quests
 volumeMultiplier=0.8
 concurrencyOverrides {maxCount=1, resolutionRule=PreventNew}
}
