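# Sound cue that plays the StrawTurn job board's deactivation sound, routed through the quests sound class.
# A single wave feeds a Modulator that adds slight random volume variation, and the cue boosts overall volume.
# It uses a shared TV prop attenuation asset for distance falloff.
# A local concurrency override caps simultaneous plays, taking precedence over the assigned one-per-owner concurrency set.
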SoundCue /StrawTurn/Audio/StrawTurn_JobBoard_Deactivate_Cue
root=Modulator_1
nodes [
 {id=Modulator_1, children [WavePlayer_1], properties {volumeMax=1.02, volumeMin=0.98}}
 {id=WavePlayer_1, properties {soundWaveAssetPtr=/StrawTurn/Audio/Waves/CH5S4_Sawtrurn_JobBoard_Deactivate_01}}
]
properties {
 attenuationSettings=/Game/Sounds/Attenuation/Fort_Prop_TV_Athena_Att
 bOverrideConcurrency=true
 maxDistance=500
 soundClassObject=&Fort_Quests
 volumeMultiplier=2
 attenuationOverrides {falloffDistance=1600}
 concurrencyOverrides {maxCount=2}
 concurrencySet [/Game/Sounds/Concurrency_Rules/OneMaximum_Owner_StopOldestRelease]
}
